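; Parent Blueprint for the Season 17 Battle Pass alien artifact quest actors, tinting each artifact's effects from a single FX Color variable.
; After the parent Construction Script runs, it passes that color to the N_Ambient Niagara component as ArtifactColor.
; It also converts the color to a vector and drives the P_PostInteractionAmbient particle parameters, scaling sweep, spark, flash and corona colors by different brightness multipliers.
; Event Tick is an unused disabled stub.
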(event EventTick
  (params (DeltaSeconds :type Float%20%28single-precision%29))
  (|CallParentFunction DeltaSeconds) ; This node is disabled and will not be called.\nDrag off pins to build functionality.
)

(fn ConstructionScript
  (bind _fx_color (Variables|Default|GetFX%20Color))
  (bind _returnvalue (Math|Conversions|ToVector(LinearColor) _fx_color))
  (bind _p_postinteractionambient (Variables|Components|GetP_PostInteractionAmbient))
  (|CallParentFunction)
  (Niagara|SetNiagaraVariable(LinearColor) (Variables|Components|GetN_Ambient) "ArtifactColor" _fx_color)
  (Effects|Components|ParticleSystem|SetVectorParameter _p_postinteractionambient "SweepColor" (* _returnvalue 1.0))
  (Effects|Components|ParticleSystem|SetVectorParameter _p_postinteractionambient "WideSparkColor" (* _returnvalue 2.0))
  (Effects|Components|ParticleSystem|SetVectorParameter _p_postinteractionambient "FlashColor" (* _returnvalue 2.0))
  (Effects|Components|ParticleSystem|SetVectorParameter _p_postinteractionambient "CoronaColor" (* _returnvalue 1.5))
  (Effects|Components|ParticleSystem|SetVectorParameter _p_postinteractionambient "ResidualSparksColor" (* _returnvalue 1.5)))
